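# Material function for UI materials that color-corrects an input color, exposing lift, gamma, gain, contrast, saturation, hue shift and brightness controls.
# Lift is weighted by the inverted color, then gain multiplies and inverse gamma applies as a power.
# The result passes through SCurve contrast, Desaturation driven by negated Saturation, HueShift and a brightness multiply.
# Effect Amount, as a clamped percentage, blends from the original color to the corrected one.
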MaterialFunction /Game/Athena/UI/Materials/MF_UI_ColorCorrector
graph {nodes [
  n1 = FunctionInput() {
   bUsePreviewValueAsDefault=true
   desc="UIMin=0, UIMax=1, ClampMin=0, ClampMax=1"
   inputName=Hue Shift
   inputType=FunctionInput_Scalar
   sortPriority=9
  }
  n2 = FunctionInput() {bUsePreviewValueAsDefault=true, description=Input color to apply correction, inputName=Color, sortPriority=1, previewValue=<1 1 1 1>}
  n3 = Multiply(A=FunctionInput() {bUsePreviewValueAsDefault=true, description=3 Way Lift Control, inputName=Lift, sortPriority=20}, B=OneMinus(None=@n2))
  n4 = FunctionInput() {bUsePreviewValueAsDefault=true, description=3 Way Color Gain Control, inputName=Gain, sortPriority=40, previewValue=<1 1 1 1>}
  n5 = FunctionInput() {bUsePreviewValueAsDefault=true, description=3 Way Gamma Control, inputName=Gamma, sortPriority=30, previewValue=<1 1 1 1>}
  n6 = Power(Base=Multiply(A=Add(A=@n3, B=@n2), B=@n4), Exp=Divide(B=@n5))
  n7 = FunctionInput() {
   bUsePreviewValueAsDefault=true
   desc="UIMin=-2, UIMax=2, ClampMin=-10, ClampMax=10"
   inputName=Contrast
   inputType=FunctionInput_Scalar
   sortPriority=5
   previewValue=<1 0 0 1>
  }
  n8 = Constant()
  n9 = FunctionInput() {
   bUsePreviewValueAsDefault=true
   desc="UIMin=-1, UIMax=1, ClampMin=-1, ClampMax=1"
   inputName=Saturation
   inputType=FunctionInput_Scalar
   sortPriority=8
  }
  n10 = MaterialFunctionCall(
   Hue Shift Percentage=@n1
   Texture=Desaturation(
    None=MaterialFunctionCall(In=@n6, Power=@n7) {materialFunction=&SCurve}.Result
    Fraction=Subtract(A=@n8, B=@n9)
   )
  ) {materialFunction=&HueShift}
  n11 = FunctionInput() {
   bUsePreviewValueAsDefault=true
   desc="UIMin=0, UIMax=2, ClampMin=0, ClampMax=10"
   inputName=Brightness
   inputType=FunctionInput_Scalar
   sortPriority=3
   previewValue=<1 0 0 1>
  }
  n12 = FunctionInput() {
   bUsePreviewValueAsDefault=true
   desc="UIMin=0, UIMax=100, ClampMin=0, ClampMax=100"
   inputName=Effect Amount
   inputType=FunctionInput_Scalar
   sortPriority=99
   previewValue=<100 0 0 1>
  }
  n13 = Divide(A=@n12) {constB=100}
  FunctionOutput(None=LinearInterpolate(A=NamedRerouteUsage(), B=Multiply(A=@n10.Result, B=@n11), Alpha=Saturate(None=@n13)))
  NamedRerouteDeclaration(None=@n2) {name=Texture, nodeColor=<rgba 1 0 0.6797 1>}
 ]}
