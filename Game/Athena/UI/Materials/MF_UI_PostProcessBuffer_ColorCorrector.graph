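# UI material function that samples the Slate post-process buffer at viewport UVs and color-corrects it through MF_UI_ColorCorrector.
# Exposes lift, gamma, gain, brightness, contrast, saturation, hue shift and effect amount as inputs with preset defaults.
# The Linear static switch reads LinearRGB on both branches, so it changes nothing.
# HSL-derived, S-curved saturation and lightness maps are built, but the multiplies with Saturation and Brightness never reach the output.
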MaterialFunction /Game/Athena/UI/Materials/MF_UI_PostProcessBuffer_ColorCorrector
graph {nodes [
  n1 = FunctionInput(Preview=Constant() {r=1.5}) {bUsePreviewValueAsDefault=true, inputName=Brightness, inputType=FunctionInput_Scalar, sortPriority=5}
  n2 = Constant() {r=0.5}
  n3 = FunctionInput(Preview=Constant() {r=0.5}) {bUsePreviewValueAsDefault=true, inputName=Saturation, inputType=FunctionInput_Scalar, sortPriority=4}
  n4 = Constant()
  n5 = Constant3Vector() {constant=<rgba 0.0015 0.003 0.0023 1>}
  n6 = Constant3Vector() {constant=<rgba 1.55 1.45 1.666 1>}
  n7 = Constant3Vector() {constant=<rgba 2 2 2.25 1>}
  n8 = Constant() {r=100}
  FunctionOutput(None=MaterialFunctionCall!(
    Color=NamedRerouteUsage()
    Brightness=@n1
    Contrast=FunctionInput(Preview=@n2) {bUsePreviewValueAsDefault=true, inputName=Contrast, inputType=FunctionInput_Scalar, sortPriority=7}
    Saturation=@n3
    Hue Shift=FunctionInput(Preview=@n4) {bUsePreviewValueAsDefault=true, inputName=HueShift, inputType=FunctionInput_Scalar, sortPriority=6}
    Lift=FunctionInput(Preview=@n5) {bUsePreviewValueAsDefault=true, inputName=Lift}
    Gamma=FunctionInput(Preview=@n6) {bUsePreviewValueAsDefault=true, inputName=Gamma, sortPriority=1}
    Gain=FunctionInput(Preview=@n7) {bUsePreviewValueAsDefault=true, inputName=Gain, sortPriority=2}
    Effect Amount=FunctionInput(Preview=@n8) {bUsePreviewValueAsDefault=true, inputName=EffectPercent, inputType=FunctionInput_Scalar, sortPriority=8}
   ) {materialFunction=&/MF_UI_ColorCorrector}.Result)
  Multiply(A=@n3, B=NamedRerouteUsage())
  Multiply(A=@n1, B=NamedRerouteUsage())
  n9 = Multiply(A=Constant() {r=0.5}) {constB=-1}
  n10 = MaterialFunctionCall!("UVs (Optional)"=ScreenPosition().ViewportUV) {materialFunction=/Engine/Functions/UserInterface/GetSlatePost0}
  n11 = StaticSwitch(
   True=@n10.LinearRGB
   False=@n10.LinearRGB
   Value=FunctionInput(Preview=StaticBool() {value=true}) {bUsePreviewValueAsDefault=true, inputName=Linear, inputType=FunctionInput_StaticBool, sortPriority=100}
  )
  n12 = MaterialFunctionCall(Float3=MaterialFunctionCall!(RGB=NamedRerouteDeclaration(None=@n11) {name=BufferRGB, nodeColor=<rgba 0 0 0 1>}) {materialFunction=/Game/ArtTools/ColorBreakdown/ColorBreakdownModes/MaterialFunctions/MF_RGBtoHSL}.Result) {materialFunction=&BreakOutFloat3Components}
  n13 = OneMinus(None=@n12.G)
  NamedRerouteDeclaration(None=MaterialFunctionCall(In=SmoothStep(Min=@n9, Value=@n13), Power=ScalarParameter() {defaultValue=1, parameterName=SaturationCurveAdjust}) {materialFunction=&SCurve}.Result) {name=SaturationMap, nodeColor=<rgba 0 0 0 1>}
  n14 = OneMinus(None=@n12.B)
  NamedRerouteDeclaration(None=MaterialFunctionCall(In=@n14, Power=ScalarParameter() {defaultValue=1, parameterName=LightnessCurveAdjust}) {materialFunction=&SCurve}.Result) {name=LightnessMap, nodeColor=<rgba 0 0 0 1>}
  Subtract(B=@n14)
 ]}
