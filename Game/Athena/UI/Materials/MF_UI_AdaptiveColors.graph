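# UI material function that derives a palette of adaptive colors from the Slate post-process buffer, exposed as separate outputs.
# Color corrector calls with preset saturation, brightness, hue and contrast produce Mid Grey, Dim, Dark, Medium, Bright and White.
# Base passes through Slate Post 0; BlueGlow linear-dodges it with blue, then lerps from blue.
# A LinearColorSpace static bool feeds each corrector and the Base switch, whose branches both read LinearRGB.
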MaterialFunction /Game/Athena/UI/Materials/MF_UI_AdaptiveColors
graph {
 properties {previewMaterialDomain=MD_UI}
 nodes [
  FunctionOutput(None=MaterialFunctionCall!(
    Lift
    Gamma
    Gain
    Saturation=Constant() {r=-0.75}
    Brightness=Constant() {r=0.5}
    HueShift=Constant() {r=0.5}
    Contrast=Constant() {r=1}
    EffectPercent=Constant() {r=100}
    Linear=NamedRerouteUsage()
   ) {materialFunction=&/MF_UI_PostProcessBuffer_ColorCorrector}.Result) {outputName=Mid Grey, sortPriority=1}
  FunctionOutput(None=MaterialFunctionCall(
    Saturation=Constant() {r=0.25}
    Brightness=Constant() {r=1}
    HueShift=Constant() {r=-0.02}
    Contrast=Constant() {r=1}
    EffectPercent=Constant() {r=100}
    Linear=NamedRerouteUsage()
   ) {materialFunction=&MF_UI_PostProcessBuffer_ColorCorrector}.Result) {outputName=Dim, sortPriority=2}
  FunctionOutput(None=MaterialFunctionCall(
    Saturation=Constant() {r=0.75}
    Brightness=Constant() {r=0.075}
    HueShift=Constant() {r=0.02}
    Contrast=Constant() {r=1.25}
    EffectPercent=Constant() {r=100}
    Linear=NamedRerouteUsage()
   ) {materialFunction=&MF_UI_PostProcessBuffer_ColorCorrector}.Result) {outputName=Dark}
  FunctionOutput(None=MaterialFunctionCall(
    Saturation=Constant()
    Brightness=Constant() {r=1.3333}
    HueShift=Constant()
    Contrast=Constant() {r=0.5}
    EffectPercent=Constant() {r=100}
    Linear=NamedRerouteUsage()
   ) {materialFunction=&MF_UI_PostProcessBuffer_ColorCorrector}.Result) {outputName=Medium, sortPriority=3}
  FunctionOutput(None=MaterialFunctionCall(
    Saturation=Constant()
    Brightness=Constant() {r=2.5}
    HueShift=Constant() {r=-0.02}
    Contrast=Constant() {r=0.75}
    EffectPercent=Constant() {r=100}
    Linear=NamedRerouteUsage()
   ) {materialFunction=&MF_UI_PostProcessBuffer_ColorCorrector}.Result) {outputName=Bright, sortPriority=4}
  FunctionOutput(None=MaterialFunctionCall(
    Lift=Constant3Vector() {constant=<rgba 1 1 1 1>}
    Gamma=Constant3Vector() {constant=<rgba 1 1 1 1>}
    Gain=Constant3Vector() {constant=<rgba 1 1 1 1>}
    Saturation=Constant()
    Brightness=Constant() {r=1}
    HueShift=Constant() {r=-0.05}
    Contrast=Constant() {r=0.5}
    EffectPercent=Constant() {r=100}
    Linear=NamedRerouteUsage()
   ) {materialFunction=&MF_UI_PostProcessBuffer_ColorCorrector}.Result) {outputName=White, sortPriority=5}
  n1 = Constant4Vector() {constant=<rgba 0 0 1 1>}
  n2 = MaterialFunctionCall!("UVs (Optional)") {materialFunction=/Engine/Functions/UserInterface/GetSlatePost0}
  n3 = StaticSwitch(True=@n2.LinearRGB, False=@n2.LinearRGB, Value=NamedRerouteUsage())
  n4 = Saturate(None=MaterialFunctionCall!(Blend=@n1.RGB, Base=FunctionOutput(None=@n3) {outputName=Base}) {materialFunction=/Engine/Functions/Engine_MaterialFunctions03/Blends/Blend_LinearDodge}.Result)
  n5 = Constant() {r=0.75}
  FunctionOutput(None=Saturate(None=LinearInterpolate(A=@n1.RGB, B=@n4, Alpha=@n5))) {outputName=BlueGlow, sortPriority=5}
  NamedRerouteDeclaration(None=FunctionInput(Preview=StaticBool() {value=true}) {bUsePreviewValueAsDefault=true, inputName=LinearColorSpace, inputType=FunctionInput_StaticBool, sortPriority=100}) {name=LinearColorSpaceValue, nodeColor=<rgba 0 0 0 1>}
 ]
}
