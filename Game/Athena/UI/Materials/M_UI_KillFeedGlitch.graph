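# Translucent UI material, named for the kill feed, that builds a glitching animation of a square and glyph icons.
# Aspect-corrected UVs feed a procedural square mask and two CuriousMarks textures, with sine timers and If thresholds switching and blending between them.
# Atlas hologram wiggle feeds the scanline function's block opacity, breaking up the square and driving further flicker branches.
# The graph shows no connections to the material outputs.
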Material /Game/Athena/UI/Materials/M_UI_KillFeedGlitch
graph {
 properties {
  bUsedWithStaticMesh=false
  blendMode=BLEND_Translucent
  materialDomain=MD_UI
  pixelDepthOffsetMode=PDOM_Legacy
  usedShadingModels=Default Lit
 }
 nodes [
  Constant() {r=120}
  Constant() {r=0.75}
  Constant() {r=0.7}
  Constant() {r=1}
  Constant() {r=0.8}
  n1 = MaterialFunctionCall(Speed=Constant() {r=4}) {materialFunction=&DebugTimeSine}
  n2 = Constant() {r=0.7}
  n3 = TextureCoordinate()
  n4 = MaterialFunctionCall() {materialFunction=&GetUserInterfaceUV}
  n5 = MaterialFunctionCall(X=Divide(A=ComponentMask(None=@n4.Pixel Size) {r=true}, B=ComponentMask(None=@n4.Pixel Size) {g=true}), Y=Constant() {r=1}) {materialFunction=&MakeFloat2}
  n6 = Multiply(A=@n3, B=@n5.Result)
  n7 = MaterialFunctionCall(Float2=@n6) {materialFunction=&BreakOutFloat2Components}
  n8 = MaterialFunctionCall(X=Add(A=@n7.R, B=ScalarParameter() {parameterName=xcv}), Y=@n7.G) {materialFunction=&MakeFloat2}
  n9 = TextureSample(UVs=@n8.Result) {samplerSource=SSM_Clamp_WorldGroupSettings, texture=/Game/Athena/UI/Icons/T_CuriousMarks_Masked_A}
  n10 = MaterialFunctionCall(Speed=Constant() {r=8}) {materialFunction=&DebugTimeSine}
  n11 = Constant() {r=0.3}
  n12 = TextureSample(UVs=@n8.Result) {samplerSource=SSM_Clamp_WorldGroupSettings, texture=/Game/Athena/UI/Icons/T_CuriousMarks_Masked_B}
  n13 = Constant() {r=0.4}
  n14 = Add(
   A=MaterialFunctionCall(UVs=@n6, Texture Scale=Constant2Vector() {g=0.5, r=0.5}) {materialFunction=&ScaleUVsByCenter}.UVs
   B=Constant2Vector()
  )
  n15 = Abs(None=Multiply(A=Subtract(A=@n14) {constB=0.5}) {constB=2})
  n16 = Max(A=ComponentMask(None=@n15) {r=true}, B=ComponentMask(None=@n15) {g=true})
  n17 = ScalarParameter() {defaultValue=-1, parameterName=xxc}
  n18 = OneMinus(None=Floor(None=Add(A=@n16, B=@n17) {constB=0.2}))
  n19 = MaterialFunctionCall(Speed=Constant() {r=0.2}) {materialFunction=&DebugTimeSine}
  n20 = MaterialFunctionCall!(
   WPO Intensity A=VectorParameter() {parameterName=wiggle1, defaultValue=<rgba 0.2 0.4 0.3524 1>}.RGB
   WPO Intensity B=VectorParameter() {parameterName=wiggle1_1, defaultValue=<rgba 0 0.7333 0 0.4>}.RGB
   Time=Multiply(A=Time()) {constB=0.05}
   Bias
  ) {materialFunction=/Game/Effects/Fort_Effects/Effects/Atlas/Materials/MF_Atlas_HologramWiggle_01}
  n21 = MaterialFunctionCall!(
   Number of Scanlines=LinearInterpolate(Alpha=@n19.Result) {constA=5, constB=10}
   Scanline Speed
   "Scanlines Thickness (Bias)"=Constant() {r=-0.2}
   Scanlines Dark Value=Constant()
   Scanlines Light Value=Constant() {r=1}
   Block Opacity=@n20.Intense 0-1
   UVs
   Time
   Block Repeat=Constant() {r=2}
  ) {materialFunction=/Game/Effects/Fort_Effects/Effects/Atlas/Materials/MF_Atlas_Scanlines_01}
  n22 = Multiply(A=@n21.Result)
  n23 = ScalarParameter() {defaultValue=2, parameterName=CubeToGlyphSpeed}
  n24 = If(
   A=MaterialFunctionCall() {materialFunction=&DebugTimeSine}.Result
   B=LinearInterpolate(
    A=Constant() {r=0.2}
    B=Constant() {r=0.8}
    Alpha=MaterialFunctionCall(Speed=@n23) {materialFunction=&DebugTimeSine}.Result
   )
   "A > B"=Constant() {r=1}
   "A == B"=Constant() {r=0.5}
   "A < B"=Constant()
  )
  If(
   A=MaterialFunctionCall(Speed=ScalarParameter() {defaultValue=0.8, parameterName=GlyphShuffleSpeed}) {materialFunction=&DebugTimeSine}.Result
   B=Constant() {r=0.8}
   "A > B"=LinearInterpolate(
    A=If(A=@n1.Result, B=@n2, "A > B"=@n9.R, "A == B"=@n9.R, "A < B"=@n9.R)
    B=If(A=@n10.Result, B=@n11, "A > B"=@n12.R, "A == B"=@n12.R, "A < B"=@n12.R)
    Alpha=MaterialFunctionCall(Speed=@n13) {materialFunction=&DebugTimeSine}.Result
   )
   "A == B"=@n12.A
   "A < B"=Multiply(A=@n18, B=LinearInterpolate(B=@n22, Alpha=@n24) {constA=1})
  )
  n25 = Constant3Vector() {constant=<rgba 0.799 0.799 0.799 0>}
  n26 = Constant() {r=1.5}
  LinearInterpolate(
   A=@n25
   B=Constant3Vector()
   Alpha=If(
    A=Constant() {r=0.2}
    B=MaterialFunctionCall(Speed=@n26) {materialFunction=&DebugTimeSine}.Result
    "A > B"=Constant() {r=1}
    "A < B"=Constant()
   )
  )
  MaterialFunctionCall(UVs=@n3, Texture Scale=@n5.Result) {materialFunction=&ScaleUVsByCenter}
  n27 = MaterialFunctionCall(Speed=Constant() {r=1.5}) {materialFunction=&DebugTimeSine}
  MaterialFunctionCall!(
   TimeSpeed=Constant() {r=1}
   Jump Amount Per Step=Constant2Vector() {r=0.15}
   Number of Steps Per Unit of Time=MaterialFunctionCall(X=LinearInterpolate(Alpha=@n27.Result) {constA=1, constB=3}, Y=Constant()) {materialFunction=&MakeFloat2}.Result
   Time Offset
  ) {materialFunction=/Engine/Functions/Engine_MaterialFunctions02/SteppingPannerTime}
  MaterialFunctionCall(Speed=Constant() {r=0.6}) {materialFunction=&DebugTimeSine}
  n28 = MaterialFunctionCall(Speed=LinearInterpolate(Alpha=@n19.Result) {constB=0.8}) {materialFunction=&DebugTimeSine}
  n29 = Constant() {r=0.5}
  n30 = Constant() {r=1}
  n31 = Constant() {r=0.3}
  n32 = Constant()
  n33 = Constant()
  n34 = LinearInterpolate(
   A=@n18
   B=LinearInterpolate(Alpha=If(A=@n28.Result, B=@n29, "A > B"=@n30, "A == B"=@n31, "A < B"=@n32)) {constA=80, constAlpha=0, constB=120}
   Alpha=If(A=@n27.Result, B=LinearInterpolate(Alpha=@n27.Result) {constA=0.05, constB=0.01}, "A > B"=Constant() {r=1}, "A == B"=@n33, "A < B"=@n33)
  ) {constA=80, constAlpha=0, constB=120}
  n35 = Multiply(A=@n21.Secondary Lines, B=Multiply(A=@n34, B=Multiply(A=@n21.Result)))
  n36 = Multiply(A=@n25, B=Clamp(None=Add(A=@n34, B=@n35)))
  n37 = Multiply(A=@n20.Result)
  n38 = ComponentMask(None=@n37) {b=true}
  Multiply(A=Add(A=Multiply(B=@n36) {constA=1}, B=@n38), B=@n38)
  Multiply(A=Multiply(A=@n36) {constB=20}, B=ComponentMask(None=@n37) {g=true})
  Multiply(A=@n20.Intense 0-1)
  ScalarParameter() {defaultValue=-1, parameterName=xxc_1}
 ]
}
