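# M_UI_ShineOutline is a translucent UI material that sweeps a shine texture diagonally across an element.
# It slightly rotates the UVs, then offsets both axes equally to move the clamped shine pattern.
# A static switch drives that offset from either looping time or the clamped ManualAnimateScalar parameter.
# The sample feeds a linear-dodge blend over Color, a ShineTint multiply and a clamped red-channel mask.
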Material /Game/Athena/UI/Materials/M_UI_ShineOutline
graph {
 properties {
  bUsedWithStaticMesh=false
  blendMode=BLEND_Translucent
  materialDomain=MD_UI
  pixelDepthOffsetMode=PDOM_Legacy
  usedShadingModels=Default Lit
 }
 nodes [
  n1 = TextureCoordinate()
  n2 = Constant() {r=-0.05}
  n3 = ScalarParameter() {parameterName=ManualAnimateScalar}
  n4 = Multiply(A=Time() {period=1}) {constB=0.2}
  n5 = Constant() {r=1}
  n6 = StaticSwitchParameter(
   True=LinearInterpolate(Alpha=Clamp(None=@n3)) {constA=1, constB=-10}
   False=LinearInterpolate(Alpha=Fmod(A=@n4, B=@n5)) {constA=1, constB=-10}
  ) {parameterName=ManuallyAnimate}
  n7 = TextureSample(UVs=Add(
    A=MaterialFunctionCall(UVs=@n1, "Rotation Angle (0-1)"=@n2) {materialFunction=&CustomRotator}.Rotated Values
    B=AppendVector(A=@n6, B=@n6)
   )) {samplerSource=SSM_Clamp_WorldGroupSettings, texture=/Game/UI/Foundation/Textures/Pattern/T-UI-Shine}
  Clamp(None=@n7.R)
  MaterialFunctionCall!(Blend=@n7.RGB, Base=VectorParameter() {parameterName=Color, defaultValue=<rgba 0 0 0 1>}.RGB) {materialFunction=/Engine/Functions/Engine_MaterialFunctions03/Blends/Blend_LinearDodge}
  Multiply(A=@n7.RGB, B=VectorParameter() {parameterName=ShineTint, defaultValue=<rgba 0.9936 1 0 1>}.RGB)
 ]
}
