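# Widget for one modal list entry: a skewed-material pip image beside auto-wrapping rich text using the styled modal text style set.
# A SetText event writes the description and stores it in the text variable, which PreConstruct applies.
# ConvertToOldStyle restyles the entry by tinting pip and text light blue, shearing the text and aligning the row top-left.
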WidgetBlueprint /Game/Athena/UI/GenericModalEntry
graphs [
 ConvertToOldStyle {nodes [
   ConvertToOldStyle() {is=title K2Node_FunctionEntry}
   > SetColorAndOpacity(self:Image Object Reference=Get Pip() -> (Pip:Image Object Reference), InColorAndOpacity:Linear Color Structure=(R=0.2122,G=0.6584,B=1,A=1))
   n1 = Get Description() -> (Description:Common Rich Text Block Object Reference)
   > Set!(
    execute
    RenderTransform_Translation:Vector 2D Structure=(X=0,Y=0)
    RenderTransform_Scale:Vector 2D Structure=(X=1,Y=1)
    RenderTransform_Shear:Vector 2D Structure=(X=-8,Y=0)
    RenderTransform_Angle:float=0
    self:Widget Object Reference=@n1
   ) -> (then, Output_Get:Widget Transform Structure)
   > SetDefaultColorAndOpacity!(
    execute
    self:Rich Text Block Object Reference=@n1
    InColorAndOpacity_SpecifiedColor:Linear Color Structure=(R=0.2118,G=0.6588,B=1,A=1)
    InColorAndOpacity_ColorUseRule:ESlateColorStylingMode Enum
   ) -> (then)
   n2 = SlotAsOverlaySlot(Widget=Get HorizontalBox_14() -> (HorizontalBox_14:Horizontal Box Object Reference))
   > SetHorizontalAlignment(self=@n2, InHorizontalAlignment=HAlign_Left)
   > SetVerticalAlignment(self=@n2, InVerticalAlignment=VAlign_Top)
  ]}
 EventGraph {nodes [
   n1 = SetText!() -> (OutputDelegate, then, InText:Text) {is=title K2Node_CustomEvent}
   > SetText(self=Get Description(), InText=@n1.InText) {is=title K2Node_CallFunction}
   > Set text(text:Text=@n1.InText) -> (Output_Get:Text)
   Event PreConstruct()
   > SetText(self=Get Description(), InText=Get text() -> (text:Text))
  ]}
]
variables [text {type=string, default="This is the <b>description</>, this is the wrapping policy when the string is long."}]
widgets {
 info {inheritedWidgetCount=0, parentClass=&UserWidget, rootWidgetClass=&Overlay, widgetCount=4}
 widgets [
  Overlay_0 {widgetClassPath=&Overlay, widget=Overlay_0}
  HorizontalBox_14 {
   widgetClassPath=&HorizontalBox
   bIsVariable=true
   parent=Overlay_0
   widget=HorizontalBox_14
   properties {flowDirectionPreference=Culture}
   slot {class=OverlaySlot, properties {horizontalAlignment=HAlign_Fill, verticalAlignment=VAlign_Fill, padding=<ltrb 0 4 0 16>}}
  }
  Pip {
   widgetClassPath=&Image
   bIsVariable=true
   parent=HorizontalBox_14
   widget=Pip
   properties {
    brush {
     bIsDynamicallyLoaded=false
     drawAs=Image
     imageType=NoImage
     mirroring=NoMirror
     resourceName=None
     resourceObject=/Game/Athena/UI/ReplayBrowser/M_Ui_Skewed_Vfor9slices_Inst_Inst
     tiling=NoTile
     imageSize=<16 16>
     margin=<ltrb 0 0 0 0>
     outlineSettings {
      bUseBrushTransparency=false
      roundingType=HalfHeightRadius
      width=0
      color {colorUseRule=UseColor_Specified, specifiedColor=<rgba 0 0 0 0>}
      cornerRadii=<0 0 0 0>
     }
     tintColor {colorUseRule=UseColor_Specified, specifiedColor=<rgba 1 1 1 1>}
     uVRegion {bIsValid=false, max=<0 0>, min=<0 0>}
    }
    colorAndOpacity=<rgba 1 1 1 0.5>
   }
   slot {class=HorizontalBoxSlot, properties {horizontalAlignment=HAlign_Left, verticalAlignment=VAlign_Center, padding=<ltrb 0 0 16 0>}}
  }
  Description {
   widgetClassPath=&CommonRichTextBlock
   bIsVariable=true
   parent=HorizontalBox_14
   widget=Description
   properties {
    autoWrapText=true
    bOverrideDefaultStyle=true
    lineHeightPercentage=0.875
    mobileTextBlockScale=1.1
    text="This is the <b>description</>, this is the wrapping policy when the string is long."
    textStyleSet=/Game/UI/Frontend/Lobby/Habanero/HabaneroData/StyledModalRichTextStyles
    defaultTextStyleOverride {
     overflowPolicy=Clip
     transformPolicy=None
     colorAndOpacity {colorUseRule=UseColor_Specified, specifiedColor=<rgba 1 1 1 1>}
     font {
      bForceMonospaced=false
      bMaterialIsStencil=false
      fontMaterial=null
      fontObject=&BurbankSmall
      letterSpacing=0
      monospacedWidth=1
      size=18
      skewAmount=0
      typefaceFontName=Bold
      outlineSettings {
       bApplyOutlineToDropShadows=false
       bMiteredCorners=false
       bSeparateFillAlpha=false
       outlineMaterial=null
       outlineSize=0
       outlineColor=<rgba 0 0 0 1>
      }
     }
     highlightColor {colorUseRule=UseColor_Specified, specifiedColor=<rgba 0 0 0 1>}
     highlightShape {
      bIsDynamicallyLoaded=false
      drawAs=Image
      imageType=NoImage
      mirroring=NoMirror
      resourceName=None
      resourceObject=null
      tiling=NoTile
      imageSize=<32 32>
      margin=<ltrb 0 0 0 0>
      outlineSettings {
       bUseBrushTransparency=false
       roundingType=HalfHeightRadius
       width=0
       color {colorUseRule=UseColor_Specified, specifiedColor=<rgba 0 0 0 0>}
       cornerRadii=<0 0 0 0>
      }
      tintColor {colorUseRule=UseColor_Specified, specifiedColor=<rgba 1 1 1 1>}
      uVRegion {bIsValid=false, max=<0 0>, min=<0 0>}
     }
     selectedBackgroundColor {colorUseRule=UseColor_Foreground, specifiedColor=<rgba 1 0 1 1>}
     shadowColorAndOpacity=<rgba 0 0 0 1>
     shadowOffset=<0 0>
     strikeBrush {
      bIsDynamicallyLoaded=false
      drawAs=Image
      imageType=NoImage
      mirroring=NoMirror
      resourceName=None
      resourceObject=null
      tiling=NoTile
      imageSize=<32 32>
      margin=<ltrb 0 0 0 0>
      outlineSettings {
       bUseBrushTransparency=false
       roundingType=HalfHeightRadius
       width=0
       color {colorUseRule=UseColor_Specified, specifiedColor=<rgba 0 0 0 0>}
       cornerRadii=<0 0 0 0>
      }
      tintColor {colorUseRule=UseColor_Specified, specifiedColor=<rgba 1 1 1 1>}
      uVRegion {bIsValid=false, max=<0 0>, min=<0 0>}
     }
     underlineBrush {
      bIsDynamicallyLoaded=false
      drawAs=Image
      imageType=NoImage
      mirroring=NoMirror
      resourceName=None
      resourceObject=null
      tiling=NoTile
      imageSize=<32 32>
      margin=<ltrb 0 0 0 0>
      outlineSettings {
       bUseBrushTransparency=false
       roundingType=HalfHeightRadius
       width=0
       color {colorUseRule=UseColor_Specified, specifiedColor=<rgba 0 0 0 0>}
       cornerRadii=<0 0 0 0>
      }
      tintColor {colorUseRule=UseColor_Specified, specifiedColor=<rgba 1 1 1 1>}
      uVRegion {bIsValid=false, max=<0 0>, min=<0 0>}
     }
    }
   }
   slot {class=HorizontalBoxSlot, properties {verticalAlignment=VAlign_Center, size {sizeRule=Fill, value=1}}}
  }
 ]
}
