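# Material function, exposed to the library in the Athena UI materials folder, that converts an HSV input into an RGB output.
# It uses a branchless method with constant offsets of 1, 2/3 and 1/3: hue is shifted per channel, wrapped with Frac, scaled, then folded with Abs and clamped into a pure hue color.
# Saturation blends from white to that color, and value multiplies the result.
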MaterialFunction /Game/Athena/UI/Materials/MF_HSVtoRGB
graph {
 properties {bExposeToLibrary=true}
 nodes [
  n1 = FunctionInput() {inputName=HSV_In}
  n2 = MaterialFunctionCall(X=Constant() {r=1}, Y=Divide(A=Constant() {r=2}, B=Constant() {r=3}), Z=Divide(A=Constant() {r=1}, B=Constant() {r=3}), A=Constant() {r=3}) {materialFunction=&MakeFloat4}
  n3 = ComponentMask(None=@n2.Result) {r=true}
  n4 = AppendVector(A=AppendVector(A=@n3, B=@n3), B=@n3)
  n5 = ComponentMask(None=@n1) {r=true}
  n6 = Add(A=AppendVector(A=AppendVector(A=@n5, B=@n5), B=@n5), B=ComponentMask(None=@n2.Result) {b=true, g=true, r=true})
  n7 = ComponentMask(None=@n2.Result) {a=true}
  n8 = Subtract(A=Multiply(A=Frac(None=@n6), B=Constant() {r=6}) {constB=6}, B=AppendVector(A=AppendVector(A=@n7, B=@n7), B=@n7))
  n9 = Clamp(None=Subtract(A=Abs(None=@n8), B=@n4))
  n10 = ComponentMask(None=@n1) {g=true}
  FunctionOutput(None=Multiply(A=ComponentMask(None=@n1) {b=true}, B=LinearInterpolate(A=@n4, B=@n9, Alpha=@n10))) {outputName=RGB_Out}
 ]
}
