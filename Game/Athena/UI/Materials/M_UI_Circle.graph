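# Translucent UI material that provides a filled circle shape with an animatable hover color.
# UVs are remapped to a centered -1 to 1 range, and distance from the center feeds an inverted SmoothStep, producing a procedural disc mask with a thin soft edge.
# A HoverAnimate scalar blends the BaseColor parameter's RGB toward HoverColor's.
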Material /Game/Athena/UI/Materials/M_UI_Circle
graph {
 properties {
  bUsedWithStaticMesh=false
  blendMode=BLEND_Translucent
  materialDomain=MD_UI
  pixelDepthOffsetMode=PDOM_Legacy
  usedShadingModels=Default Lit
 }
 nodes [
  LinearInterpolate(
   A=VectorParameter() {parameterName=BaseColor, defaultValue=<rgba 1 1 1 1>}.RGB
   B=VectorParameter() {parameterName=HoverColor, defaultValue=<rgba 1 1 1 0>}.RGB
   Alpha=ScalarParameter() {parameterName=HoverAnimate, sliderMax=1}
  )
  n1 = Distance(A=ConstantBiasScale(None=TextureCoordinate()) {bias=-0.5, scale=2}, B=Constant())
  n2 = Constant() {r=1}
  OneMinus(None=SmoothStep(Min=Constant(), Max=Constant() {r=0.01}, Value=Subtract(A=@n1, B=@n2)))
 ]
}
